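# Sound cue for the Hightower Radish pet's facepalm reaction at the start of DBNO.
# A root mixer plays a fixed facepalm wave alongside one of four randomly chosen voice lines, with the voice layer mixed much quieter.
# The fourth voice line passes through its own mixer for a volume boost.
# Playback uses the creature pets sound class and close-range pet attenuation.
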SoundCue /BRCosmetics/Sounds/Pets/Hightower_Radish/HightowerRadish_Pet_DBNOStart_FacePalm
root=Mixer_0
nodes [
 {id=Mixer_0, children [WavePlayer_0,Random_0], properties {inputVolume [1,0.35]}}
 {id=Mixer_1, children [WavePlayer_4], properties {inputVolume [1.5]}}
 {id=Random_0, children [WavePlayer_1,WavePlayer_2,WavePlayer_3,Mixer_1], properties {weights [1,1,1,1]}}
 {id=WavePlayer_0, properties {soundWaveAssetPtr=&/Pet_HightowerRadish_DBNO_Start_FacePalm_01}}
 {id=WavePlayer_1, properties {soundWaveAssetPtr=&/Pet_HightowerRadish_DBNO_Start_FacePalm_Voice_01}}
 {id=WavePlayer_2, properties {soundWaveAssetPtr=&/Pet_HightowerRadish_DBNO_Start_FacePalm_Voice_02}}
 {id=WavePlayer_3, properties {soundWaveAssetPtr=&/Pet_HightowerRadish_DBNO_Start_FacePalm_Voice_03}}
 {id=WavePlayer_4, properties {soundWaveAssetPtr=&/Pet_HightowerRadish_DBNO_Start_FacePalm_Voice_04}}
]
properties {
 attenuationSettings=&Fort_Pets_Close_Athena_Att
 maxDistance=1800.0001
 soundClassObject=&Fort_Creatures_Pets
 volumeMultiplier=0.896
}
