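# Sound cue for the Hightower Radish pet's pouting reaction at DBNO start.
# A mixer layers a base pout sound at full volume with a quieter voice layer.
# The voice layer picks one of three equally weighted voice variations at random.
# It plays through the creature pets sound class with close-range pet attenuation and a boosted overall volume.
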SoundCue /BRCosmetics/Sounds/Pets/Hightower_Radish/HightowerRadish_Pet_DBNOStart_Pout
root=Mixer_0
nodes [
 {id=Mixer_0, children [WavePlayer_3,Random_0], properties {inputVolume [1,0.2]}}
 {id=Random_0, children [WavePlayer_4,WavePlayer_5,WavePlayer_6], properties {weights [1,1,1]}}
 {id=WavePlayer_3, properties {soundWaveAssetPtr=&/Pet_HightowerRadish_DBNO_Start_Pout_01}}
 {id=WavePlayer_4, properties {soundWaveAssetPtr=&/Pet_HightowerRadish_DBNO_Start_Pout_Voice_01}}
 {id=WavePlayer_5, properties {soundWaveAssetPtr=&/Pet_HightowerRadish_DBNO_Start_Pout_Voice_02}}
 {id=WavePlayer_6, properties {soundWaveAssetPtr=&/Pet_HightowerRadish_DBNO_Start_Pout_Voice_03}}
]
properties {
 attenuationSettings=&Fort_Pets_Close_Athena_Att
 maxDistance=1800.0001
 soundClassObject=&Fort_Creatures_Pets
 volumeMultiplier=2.016
}
