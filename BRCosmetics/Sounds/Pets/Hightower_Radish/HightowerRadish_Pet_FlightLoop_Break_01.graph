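# Sound cue for the Hightower Radish pet's flight loop break, likely a brief adjustment movement paired with a vocal.
# A mixer layers a random pick of four equally weighted adjust sounds with a quieter random pick of four voice variants.
# The voice layer starts after a short fixed delay.
# It plays under the creature pets sound class with close-range pet attenuation.
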SoundCue /BRCosmetics/Sounds/Pets/Hightower_Radish/HightowerRadish_Pet_FlightLoop_Break_01
root=Mixer_0
nodes [
 {id=Delay_0, children [Random_1], properties {delayMax=0.65, delayMin=0.65}}
 {id=Mixer_0, children [Random_0,Delay_0], properties {inputVolume [1,0.6]}}
 {id=Random_0, children [WavePlayer_1,WavePlayer_2,WavePlayer_3,WavePlayer_4], properties {weights [1,1,1,1]}}
 {id=Random_1, children [WavePlayer_5,WavePlayer_6,WavePlayer_7,WavePlayer_8], properties {weights [1,1,1,1]}}
 {id=WavePlayer_1, properties {soundWaveAssetPtr=&/Pet_HightowerRadish_FlightLoop_Break01_Adjust_01}}
 {id=WavePlayer_2, properties {soundWaveAssetPtr=&/Pet_HightowerRadish_FlightLoop_Break01_Adjust_02}}
 {id=WavePlayer_3, properties {soundWaveAssetPtr=&/Pet_HightowerRadish_FlightLoop_Break01_Adjust_03}}
 {id=WavePlayer_4, properties {soundWaveAssetPtr=&/Pet_HightowerRadish_FlightLoop_Break01_Adjust_04}}
 {
  id=WavePlayer_5
  properties {soundWaveAssetPtr=&/Pet_HightowerRadish_FlightLoop_Break01_Adjust_Voice_01}
 }
 {
  id=WavePlayer_6
  properties {soundWaveAssetPtr=&/Pet_HightowerRadish_FlightLoop_Break01_Adjust_Voice_02}
 }
 {
  id=WavePlayer_7
  properties {soundWaveAssetPtr=&/Pet_HightowerRadish_FlightLoop_Break01_Adjust_Voice_03}
 }
 {
  id=WavePlayer_8
  properties {soundWaveAssetPtr=&/Pet_HightowerRadish_FlightLoop_Break01_Adjust_Voice_04}
 }
]
properties {
 attenuationSettings=&Fort_Pets_Close_Athena_Att
 maxDistance=1800.0001
 soundClassObject=&Fort_Creatures_Pets
 volumeMultiplier=0.336
}
