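# Sound cue for the HightowerRadish pet's underwater hover-idle exit, layering an effect with a voice.
# A mixer combines one of two random exit effects with one of five random voice clips, the voice sitting quieter.
# The effect branch varies pitch on each play, with an extra pitch-down modulator on the second clip.
# Playback uses close-range pet attenuation and the creature pets sound class.
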SoundCue /BRCosmetics/Sounds/Pets/Hightower_Radish/HightowerRadish_Pet_UnderwaterExit
root=Mixer_0
nodes [
 {id=Mixer_0, children [Modulator_0,Random_2], properties {inputVolume [1,0.65]}}
 {id=Modulator_0, children [Random_0], properties {pitchMax=1.1, pitchMin=0.9, volumeMax=1, volumeMin=1}}
 {id=Modulator_1, children [WavePlayer_1], properties {pitchMax=0.95, volumeMax=1, volumeMin=1}}
 {id=Random_0, children [WavePlayer_0,Modulator_1], properties {weights [1,1]}}
 {id=Random_2, children [WavePlayer_2,WavePlayer_3,WavePlayer_5,WavePlayer_6,WavePlayer_7], properties {weights [1,1,1,1,1]}}
 {id=WavePlayer_0, properties {soundWaveAssetPtr=&/Pet_HIghtowerRadish_Hover_Idle_UnderWater_Exit_01}}
 {id=WavePlayer_1, properties {soundWaveAssetPtr=&/Pet_HightowerRadish_Hover_Idle_UnderWater_Exit_02}}
 {
  id=WavePlayer_2
  properties {soundWaveAssetPtr=&/Pet_HightowerRadish_Hover_Idle_UnderWater_Exit_Voice_01}
 }
 {
  id=WavePlayer_3
  properties {soundWaveAssetPtr=&/Pet_HightowerRadish_Hover_Idle_UnderWater_Exit_Voice_02}
 }
 {
  id=WavePlayer_5
  properties {soundWaveAssetPtr=&/Pet_HightowerRadish_Hover_Idle_UnderWater_Exit_Voice_04}
 }
 {
  id=WavePlayer_6
  properties {soundWaveAssetPtr=&/Pet_HightowerRadish_Hover_Idle_UnderWater_Exit_Voice_05}
 }
 {
  id=WavePlayer_7
  properties {soundWaveAssetPtr=&/Pet_HightowerRadish_Hover_Idle_UnderWater_Exit_Voice_06}
 }
]
properties {
 attenuationSettings=&Fort_Pets_Close_Athena_Att
 maxDistance=1800.0001
 soundClassObject=&Fort_Creatures_Pets
 volumeMultiplier=1.12
}
